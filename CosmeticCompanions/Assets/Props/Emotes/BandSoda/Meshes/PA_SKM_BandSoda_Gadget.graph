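# Physics asset for the BandSoda gadget skeletal mesh, a prop in the CosmeticCompanions emote assets.
# It has a single body on the root bone, a rotated capsule whose near-zero length and large radius make it effectively a sphere.
# There are no other bodies or constraints, so it gives only a coarse bounding shape with no per-bone or jointed simulation.
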PhysicsAsset /CosmeticCompanions/Assets/Props/Emotes/BandSoda/Meshes/PA_SKM_BandSoda_Gadget
bodies {root [{shapeType=Capsule, length=0.505, radius=210.1941, rotation=<rot 0 -90 0>}]}
constraints []
